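# Conversation database for the NPC Turk, defining his greeting dialogue and the interactions it leads to.
# The Turk entry point runs one speech node with a default line, overridden by prioritized contextual lines for a first-ever meeting or for specific player outfits matched by character ID.
# The speech then links by entry tag to shared NPC quest and basic service conversations, plus a Season 25 quest conversation specific to Turk.
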ConversationDatabase /NPCLibrary/NPCs/Turk/Turk/NPCLibrary_Conversation_Turk
entry_points [{tag=Conversation.Entry.Turk, destinations [#1]}]
nodes [
 ConversationEntryPointNode_7 {
  id=#1
  class=ConversationEntryPointNode
  outputs [#2]
  properties {entryTag {tagName=Conversation.Entry.Turk}}
 }
 Speech_C_4 {
  id=#2
  class=Speech_C
  outputs [#3, #4, #5]
  properties {
   nodeName=""
   generalConfig {
    defaultMessage=Don't get tangled up in my net...
    contextualMessages [
     {
      message=Nothin' to see here.
      contextRequirements [{requirement {class=FortControllerRequirement_IsFirstEverConversationWithNPC}}]
      priority {curve {rowName=Tandem.Speech.Priority.FirstEverInteraction}}
     }
     {
      message=My nets have been empty for weeks. I need our plan to work.
      requirementMatchPolicy=RequireAny
      contextRequirements [{requirement {
         class=FortControllerRequirement_HasCID
         properties {allowedCIds [/BRCosmetics/Athena/Items/Cosmetics/Characters/CID_564_Athena_Commando_M_TacticalFisherman]}
        }}]
      priority {curve {rowName=Tandem.Speech.Priority.OutfitReaction}}
     }
     {
      message="You see Coral Buddies, I see Bars."
      requirementMatchPolicy=RequireAny
      contextRequirements [{requirement {
         class=FortControllerRequirement_HasCID
         properties {allowedCIds [/BRCosmetics/Athena/Items/Cosmetics/Characters/CID_624_Athena_Commando_M_TeriyakiWarrior]}
        }}]
      priority {curve {rowName=Tandem.Speech.Priority.OutfitReaction}}
     }
     {
      message=Is that suit waterproof?
      requirementMatchPolicy=RequireAny
      contextRequirements [{requirement {class=FortControllerRequirement_HasCID, properties {allowedCIds [/BRCosmetics/Athena/Items/Cosmetics/Characters/Character_RippedHarvester]}}}]
      priority {curve {rowName=Tandem.Speech.Priority.OutfitReaction}}
     }
     {
      message="If you need a trawler, I'm your angler."
      requirementMatchPolicy=RequireAny
      contextRequirements [{requirement {class=FortControllerRequirement_HasCID, properties {allowedCIds [/BRCosmetics/Athena/Items/Cosmetics/Characters/Character_CrispRover]}}}]
      priority {curve {rowName=Tandem.Speech.Priority.OutfitReaction}}
     }
     {
      message=Your guitar playing scares the fish.
      requirementMatchPolicy=RequireAny
      contextRequirements [{requirement {class=FortControllerRequirement_HasCID, properties {allowedCIds [/BRCosmetics/Athena/Items/Cosmetics/Characters/Character_CinderMax]}}}]
      priority {curve {rowName=Tandem.Speech.Priority.OutfitReaction}}
     }
    ]
   }
  }
 }
 ConversationLinkNode_4 {
  id=#5
  class=ConversationLinkNode
  properties {remoteEntryTag {tagName=Conversation.Entry.Season25.P01.Q01.Turk}}
 }
 ConversationLinkNode_0 {id=#4, class=ConversationLinkNode, properties {remoteEntryTag {tagName=Conversation.Entry.BasicServices}}}
 ConversationLinkNode_1 {id=#3, class=ConversationLinkNode, properties {remoteEntryTag {tagName=Conversation.Entry.NPCLibrary.Quests}}}
]
